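# Physics asset for a female medium Arctic Camo Woods player cosmetic part, covering dynamic hair, hat and fluff bones around the head.
# A head capsule anchors two chains: two box hair bodies in series, and a hat capsule carrying two spherical fluff bodies.
# Angular limits keep the hat nearly rigid, allow moderate hair swing and let the fluff tip swing most freely, with little twist throughout.
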PhysicsAsset /BRCosmetics/Characters/Player/Female/Medium/Bodies/F_MED_Arctic_Camo_Woods/Meshes/Parts/F_MED_Arctic_Camo_Woods_Physics
bodies {
 dyn_fluff [{shapeType=Sphere, radius=4.3688, center=<2.6512 -0.0176 0.0598>}]
 dyn_fluff_01 [{shapeType=Sphere, radius=0.505}]
 dyn_hair_01 [{shapeType=Box, extentX=1.3906, extentY=5.2248, extentZ=3.4217, center=<1.3836 0.2603 0>, rotation=<rot 0 -90 0>}]
 dyn_hair_02 [{shapeType=Box, extentX=1.01, extentY=4.3607, extentZ=2.3759, center=<1.4569 0 0>, rotation=<rot 0 -90 0>}]
 dyn_hat [{shapeType=Capsule, length=4.3802, radius=2.7297, center=<1.673 0.5006 0.8104>, rotation=<rot 0 -6.9427 -87.9004>}]
 head [{shapeType=Capsule, length=4.5423, radius=5.2146, center=<0.8313 2.2812 0.0003>, rotation=<rot 90 0 0>}]
}
constraints [
 {bone1Name=dyn_hair_01, bone2Name=head, swing1LimitDegrees=20, swing2LimitDegrees=30, twistLimitDegrees=5}
 {bone1Name=dyn_hair_02, bone2Name=dyn_hair_01, swing1LimitDegrees=20, swing2LimitDegrees=30, twistLimitDegrees=5}
 {bone1Name=dyn_hat, bone2Name=head, swing1LimitDegrees=5, swing2LimitDegrees=5, twistLimitDegrees=1}
 {bone1Name=dyn_fluff, bone2Name=dyn_hat, swing1LimitDegrees=15, swing2LimitDegrees=15, twistLimitDegrees=3}
 {bone1Name=dyn_fluff_01, bone2Name=dyn_fluff, swing1LimitDegrees=45, swing2LimitDegrees=45, twistLimitDegrees=3}
]
